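# Physics asset for the right-wrist bracelet of the Sandcastle Tattoos male medium character body.
# A box body on the bracelet bone is constrained to a tiny box on its MSTR bone, which serves as the anchor point, with moderate swing and tight twist limits.
# A capsule on the right hand bone provides a hand shape alongside the bracelet.
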PhysicsAsset /BRCosmetics/Characters/Player/Male/Medium/Bodies/M_MED_Sandcastle_Tattoos/Meshes/M_MED_Sandcastle_Tattoos_Bracelet_Physics
bodies {
 dyn_bracelet_01_r [{shapeType=Box, extentX=10.7823, extentY=2.3447, extentZ=10.6434, center=<0.0001 -4.7487 0>, rotation=<rot 0 -89.9992 0>}]
 dyn_bracelet_01_r_MSTR [{shapeType=Box, extentX=0.01, extentY=0.01, extentZ=0.01}]
 hand_r [{shapeType=Capsule, length=9.2409, radius=6.3833, center=<-6.1738 -0.0001 0>}]
}
constraints [{bone1Name=dyn_bracelet_01_r, bone2Name=dyn_bracelet_01_r_MSTR, swing1LimitDegrees=25, swing2LimitDegrees=25, twistLimitDegrees=5}]
