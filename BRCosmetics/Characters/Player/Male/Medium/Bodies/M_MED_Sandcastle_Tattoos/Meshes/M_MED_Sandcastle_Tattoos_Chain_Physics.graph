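# Physics asset for the Sandcastle Tattoos male body's chain, modeling a dangling chain on the right side.
# Front, middle and back strands of five limit-constrained boxes hang from tiny master anchors under a pelvis-linked chain root.
# Front and back tips join the middle tip through free-rotation constraints, tying the strands together.
# Capsules cover pelvis, spine and right leg, while bracelet and shoelaces have only tiny master anchor bodies.
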PhysicsAsset /BRCosmetics/Characters/Player/Male/Medium/Bodies/M_MED_Sandcastle_Tattoos/Meshes/M_MED_Sandcastle_Tattoos_Chain_Physics
bodies {
 calf_r [{shapeType=Capsule, length=35.9183, radius=10.5009, center=<21.0637 0.0004 -1.9273>, rotation=<rot 90 -0.0099 -0.0099>}]
 dyn_bracelet_01_r_MSTR [{shapeType=Capsule, length=0.01, radius=0.01}]
 dyn_chaindriver_bk_01_r [{shapeType=Box, extentX=5, extentY=1.5, extentZ=11, center=<1.9176 0 0>, rotation=<rot 0 0 -24.9986>}]
 dyn_chaindriver_bk_01_r_MSTR [{shapeType=Capsule, length=0.01, radius=0.01}]
 dyn_chaindriver_bk_02_r [{shapeType=Box, extentX=5, extentY=1.5, extentZ=11, center=<1.9176 0 0>, rotation=<rot 0 0 -24.9986>}]
 dyn_chaindriver_bk_03_r [{shapeType=Box, extentX=5, extentY=1.5, extentZ=11, center=<1.9176 0 0>, rotation=<rot 0 0 -24.9986>}]
 dyn_chaindriver_bk_04_r [{shapeType=Box, extentX=5, extentY=1.5, extentZ=11, center=<1.9176 0 0>, rotation=<rot 0 0 -24.9986>}]
 dyn_chaindriver_bk_05_r [{shapeType=Box, extentX=5, extentY=1.5, extentZ=11, center=<1.9176 0 0>, rotation=<rot 0 0 -24.9986>}]
 dyn_chaindriver_fr_01_r [{shapeType=Box, extentX=4, extentY=3, extentZ=12, center=<4.1075 0.0002 0>, rotation=<rot 0 0 24.9997>}]
 dyn_chaindriver_fr_01_r_MSTR [{shapeType=Capsule, length=0.01, radius=0.01}]
 dyn_chaindriver_fr_02_r [{shapeType=Box, extentX=4, extentY=3, extentZ=12, center=<4.1075 0.0002 0>, rotation=<rot 0 0 24.9997>}]
 dyn_chaindriver_fr_03_r [{shapeType=Box, extentX=4, extentY=3, extentZ=12, center=<4.1075 0.0002 0>, rotation=<rot 0 0 24.9997>}]
 dyn_chaindriver_fr_04_r [{shapeType=Box, extentX=4, extentY=3, extentZ=12, center=<4.1075 0.0002 0>, rotation=<rot 0 0 24.9997>}]
 dyn_chaindriver_fr_05_r [{shapeType=Box, extentX=4, extentY=3, extentZ=12, center=<4.1075 0.0002 0>, rotation=<rot 0 0 24.9997>}]
 dyn_chaindriver_md_01_r [{shapeType=Box, extentX=5, extentY=14, extentZ=2.5, center=<1.9625 0 0.0001>}]
 dyn_chaindriver_md_01_r_MSTR [{shapeType=Capsule, length=0.01, radius=0.01}]
 dyn_chaindriver_md_02_r [{shapeType=Box, extentX=5, extentY=14, extentZ=2.5, center=<1.9625 0 0.0001>}]
 dyn_chaindriver_md_03_r [{shapeType=Box, extentX=5, extentY=14, extentZ=2.5, center=<1.9625 0 0.0001>}]
 dyn_chaindriver_md_04_r [{shapeType=Box, extentX=5, extentY=14, extentZ=2.5, center=<1.9625 0 0.0001>}]
 dyn_chaindriver_md_05_r [{shapeType=Box, extentX=5, extentY=14, extentZ=2.5, center=<1.9625 0 0.0001>}]
 dyn_chainroot_MSTR [{shapeType=Capsule, length=0.01, radius=0.01}]
 dyn_shoelaceLeft_01_l_MSTR [{shapeType=Capsule, length=0.01, radius=0.01}]
 dyn_shoelaceLeft_01_r_MSTR [{shapeType=Capsule, length=0.01, radius=0.01}]
 dyn_shoelaceRight_01_l_MSTR [{shapeType=Capsule, length=0.01, radius=0.01}]
 dyn_shoelaceRight_01_r_MSTR [{shapeType=Capsule, length=0.01, radius=0.01}]
 pelvis [{shapeType=Capsule, length=9.9192, radius=8.9374, center=<-2.2087 -6.3453 0.0003>}]
 spine_02 [{shapeType=Capsule, length=13.9192, radius=9.6874}]
 thigh_r [{shapeType=Capsule, length=26.4183, radius=7.7509, center=<19.6966 -0.7064 0.0962>, rotation=<rot 90 0 0>}]
 thigh_twist_01_r [{shapeType=Capsule, length=24.6683, radius=9.5009, center=<-7.7126 -0.707 -0.9003>, rotation=<rot 90 0 0>}]
}
constraints [
 {bone1Name=spine_02, bone2Name=pelvis, swing1LimitDegrees=45, swing2LimitDegrees=45, twistLimitDegrees=45}
 {bone1Name=calf_r, bone2Name=thigh_r, swing1LimitDegrees=45, swing2LimitDegrees=45, twistLimitDegrees=45}
 {
  bone1Name=dyn_bracelet_01_r_MSTR
  bone2Name=spine_02
  swing1LimitDegrees=45
  swing1Motion=Free
  swing2LimitDegrees=45
  swing2Motion=Free
  twistLimitDegrees=45
  twistMotion=Free
 }
 {
  bone1Name=dyn_shoelaceLeft_01_l_MSTR
  bone2Name=pelvis
  swing1LimitDegrees=45
  swing1Motion=Free
  swing2LimitDegrees=45
  swing2Motion=Free
  twistLimitDegrees=45
  twistMotion=Free
 }
 {
  bone1Name=dyn_shoelaceRight_01_l_MSTR
  bone2Name=pelvis
  swing1LimitDegrees=45
  swing1Motion=Free
  swing2LimitDegrees=45
  swing2Motion=Free
  twistLimitDegrees=45
  twistMotion=Free
 }
 {
  bone1Name=dyn_shoelaceRight_01_r_MSTR
  bone2Name=calf_r
  swing1LimitDegrees=45
  swing1Motion=Free
  swing2LimitDegrees=45
  swing2Motion=Free
  twistLimitDegrees=45
  twistMotion=Free
 }
 {
  bone1Name=dyn_shoelaceLeft_01_r_MSTR
  bone2Name=calf_r
  swing1LimitDegrees=45
  swing1Motion=Free
  swing2LimitDegrees=45
  swing2Motion=Free
  twistLimitDegrees=45
  twistMotion=Free
 }
 {
  bone1Name=dyn_chainroot_MSTR
  bone2Name=pelvis
  swing1LimitDegrees=45
  swing1Motion=Free
  swing2LimitDegrees=45
  swing2Motion=Free
  twistLimitDegrees=45
  twistMotion=Free
 }
 {
  bone1Name=dyn_chaindriver_bk_01_r_MSTR
  bone2Name=dyn_chainroot_MSTR
  swing1LimitDegrees=45
  swing1Motion=Free
  swing2LimitDegrees=45
  swing2Motion=Free
  twistLimitDegrees=45
  twistMotion=Free
 }
 {
  bone1Name=dyn_chaindriver_fr_01_r_MSTR
  bone2Name=dyn_chainroot_MSTR
  swing1LimitDegrees=45
  swing1Motion=Free
  swing2LimitDegrees=45
  swing2Motion=Free
  twistLimitDegrees=45
  twistMotion=Free
 }
 {bone1Name=dyn_chaindriver_fr_01_r, bone2Name=dyn_chaindriver_fr_01_r_MSTR, swing1LimitDegrees=75, swing2LimitDegrees=25, twistLimitDegrees=5}
 {bone1Name=dyn_chaindriver_fr_02_r, bone2Name=dyn_chaindriver_fr_01_r, swing1LimitDegrees=25, swing2LimitDegrees=25, twistLimitDegrees=5}
 {bone1Name=dyn_chaindriver_fr_03_r, bone2Name=dyn_chaindriver_fr_02_r, swing1LimitDegrees=25, swing2LimitDegrees=25, twistLimitDegrees=5}
 {bone1Name=dyn_chaindriver_fr_04_r, bone2Name=dyn_chaindriver_fr_03_r, swing1LimitDegrees=25, swing2LimitDegrees=25, twistLimitDegrees=5}
 {bone1Name=dyn_chaindriver_fr_05_r, bone2Name=dyn_chaindriver_fr_04_r, swing1LimitDegrees=25, swing2LimitDegrees=25, twistLimitDegrees=5}
 {bone1Name=dyn_chaindriver_bk_01_r, bone2Name=dyn_chaindriver_bk_01_r_MSTR, swing1LimitDegrees=25, swing2LimitDegrees=75, twistLimitDegrees=5}
 {bone1Name=dyn_chaindriver_bk_02_r, bone2Name=dyn_chaindriver_bk_01_r, swing1LimitDegrees=25, swing2LimitDegrees=75, twistLimitDegrees=5}
 {bone1Name=dyn_chaindriver_bk_03_r, bone2Name=dyn_chaindriver_bk_02_r, swing1LimitDegrees=25, swing2LimitDegrees=25, twistLimitDegrees=5}
 {bone1Name=dyn_chaindriver_bk_04_r, bone2Name=dyn_chaindriver_bk_03_r, swing1LimitDegrees=25, swing2LimitDegrees=25, twistLimitDegrees=5}
 {bone1Name=dyn_chaindriver_bk_05_r, bone2Name=dyn_chaindriver_bk_04_r, swing1LimitDegrees=25, swing2LimitDegrees=25, twistLimitDegrees=5}
 {
  bone1Name=dyn_chaindriver_md_01_r_MSTR
  bone2Name=dyn_chainroot_MSTR
  swing1LimitDegrees=45
  swing1Motion=Free
  swing2LimitDegrees=45
  swing2Motion=Free
  twistLimitDegrees=45
  twistMotion=Free
 }
 {bone1Name=dyn_chaindriver_md_01_r, bone2Name=dyn_chaindriver_md_01_r_MSTR, swing1LimitDegrees=25, swing2LimitDegrees=25, twistLimitDegrees=5}
 {bone1Name=dyn_chaindriver_md_02_r, bone2Name=dyn_chaindriver_md_01_r, swing1LimitDegrees=25, swing2LimitDegrees=25, twistLimitDegrees=5}
 {bone1Name=dyn_chaindriver_md_03_r, bone2Name=dyn_chaindriver_md_02_r, swing1LimitDegrees=25, swing2LimitDegrees=25, twistLimitDegrees=5}
 {bone1Name=dyn_chaindriver_md_04_r, bone2Name=dyn_chaindriver_md_03_r, swing1LimitDegrees=25, swing2LimitDegrees=25, twistLimitDegrees=5}
 {bone1Name=dyn_chaindriver_md_05_r, bone2Name=dyn_chaindriver_md_04_r, swing1LimitDegrees=75, swing2LimitDegrees=75, twistLimitDegrees=5}
 {
  bone1Name=dyn_chaindriver_bk_05_r
  bone2Name=dyn_chaindriver_md_05_r
  swing1LimitDegrees=45
  swing1Motion=Free
  swing2LimitDegrees=45
  swing2Motion=Free
  twistLimitDegrees=45
  twistMotion=Free
 }
 {
  bone1Name=dyn_chaindriver_fr_05_r
  bone2Name=dyn_chaindriver_md_05_r
  swing1LimitDegrees=45
  swing1Motion=Free
  swing2LimitDegrees=45
  swing2Motion=Free
  twistLimitDegrees=45
  twistMotion=Free
 }
 {
  bone1Name=thigh_twist_01_r
  bone2Name=thigh_r
  swing1LimitDegrees=45
  swing1Motion=Free
  swing2LimitDegrees=45
  swing2Motion=Free
  twistLimitDegrees=45
  twistMotion=Free
 }
 {
  bone1Name=thigh_r
  bone2Name=pelvis
  swing1LimitDegrees=45
  swing1Motion=Free
  swing2LimitDegrees=45
  swing2Motion=Free
  twistLimitDegrees=45
  twistMotion=Free
 }
]
