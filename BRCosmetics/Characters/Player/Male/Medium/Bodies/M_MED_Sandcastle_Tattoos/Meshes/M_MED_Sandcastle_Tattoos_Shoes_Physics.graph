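# Physics asset for the shoes of the Sandcastle Tattoos medium male player body cosmetic, covering dynamic boot and shoelace bones.
# Each shoe has two shoelace chains of two capsules with limited swing and twist, rooted on tiny MSTR anchor bodies.
# A boot capsule on each side is tightly constrained to its own MSTR anchor.
# Capsules on both feet have no constraints, likely serving as collision for the dynamic parts.
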PhysicsAsset /BRCosmetics/Characters/Player/Male/Medium/Bodies/M_MED_Sandcastle_Tattoos/Meshes/M_MED_Sandcastle_Tattoos_Shoes_Physics
bodies {
 dyn_boot_01_l [{shapeType=Capsule, length=9.2854, radius=5.0665, center=<-1.4325 0 0>, rotation=<rot 90 -2.5305 -2.5305>}]
 dyn_boot_01_l_MSTR [{shapeType=Capsule, length=0.01, radius=0.01}]
 dyn_boot_01_r [{shapeType=Capsule, length=9.2854, radius=5.0665, center=<-1.4325 0 0>, rotation=<rot 90 -2.5305 -2.5305>}]
 dyn_boot_01_r_MSTR [{shapeType=Capsule, length=0.01, radius=0.01}]
 dyn_shoelaceLeft_01_l [{shapeType=Capsule, length=3.6743, radius=1.8572, center=<2.0392 0.0001 0>, rotation=<rot -90 0.0197 -0.0196>}]
 dyn_shoelaceLeft_01_l_MSTR [{shapeType=Capsule, length=0.01, radius=0.01}]
 dyn_shoelaceLeft_01_r [{shapeType=Capsule, length=3.6743, radius=1.8572, center=<2.0392 0.0001 0>, rotation=<rot -90 0.0197 -0.0196>}]
 dyn_shoelaceLeft_01_r_MSTR [{shapeType=Capsule, length=0.01, radius=0.01}]
 dyn_shoelaceLeft_02_l [{shapeType=Capsule, length=3.6743, radius=1.8572, center=<2.0392 0.0001 0>, rotation=<rot -90 0.0197 -0.0196>}]
 dyn_shoelaceLeft_02_r [{shapeType=Capsule, length=3.6743, radius=1.8572, center=<2.0392 0.0001 0>, rotation=<rot -90 0.0197 -0.0196>}]
 dyn_shoelaceRight_01_l [{shapeType=Capsule, length=3.6743, radius=1.8572, center=<2.0392 0.0001 0>, rotation=<rot -90 0.0197 -0.0196>}]
 dyn_shoelaceRight_01_l_MSTR [{shapeType=Capsule, length=0.01, radius=0.01}]
 dyn_shoelaceRight_01_r [{shapeType=Capsule, length=3.6743, radius=1.8572, center=<2.0392 0.0001 0>, rotation=<rot -90 0.0197 -0.0196>}]
 dyn_shoelaceRight_01_r_MSTR [{shapeType=Capsule, length=0.01, radius=0.01}]
 dyn_shoelaceRight_02_l [{shapeType=Capsule, length=3.6743, radius=1.8572, center=<2.0392 0.0001 0>, rotation=<rot -90 0.0197 -0.0196>}]
 dyn_shoelaceRight_02_r [{shapeType=Capsule, length=3.6743, radius=1.8572, center=<2.0392 0.0001 0>, rotation=<rot -90 0.0197 -0.0196>}]
 foot_l [{shapeType=Capsule, length=13.9433, radius=5.6661, center=<-2.7815 7.6421 -0.0002>, rotation=<rot 0 20 -90>}]
 foot_r [{shapeType=Capsule, length=13.9433, radius=5.6661, center=<2.7815 -7.6421 0.0002>, rotation=<rot 0 20 90>}]
}
constraints [
 {bone1Name=dyn_shoelaceLeft_01_l, bone2Name=dyn_shoelaceLeft_01_l_MSTR, swing1LimitDegrees=15, swing2LimitDegrees=15, twistLimitDegrees=5}
 {bone1Name=dyn_shoelaceLeft_02_l, bone2Name=dyn_shoelaceLeft_01_l, swing1LimitDegrees=15, swing2LimitDegrees=15, twistLimitDegrees=5}
 {bone1Name=dyn_shoelaceRight_01_l, bone2Name=dyn_shoelaceRight_01_l_MSTR, swing1LimitDegrees=15, swing2LimitDegrees=15, twistLimitDegrees=5}
 {bone1Name=dyn_shoelaceRight_02_l, bone2Name=dyn_shoelaceRight_01_l, swing1LimitDegrees=15, swing2LimitDegrees=15, twistLimitDegrees=5}
 {bone1Name=dyn_shoelaceLeft_01_r, bone2Name=dyn_shoelaceLeft_01_r_MSTR, swing1LimitDegrees=15, swing2LimitDegrees=15, twistLimitDegrees=5}
 {bone1Name=dyn_shoelaceLeft_02_r, bone2Name=dyn_shoelaceLeft_01_r, swing1LimitDegrees=15, swing2LimitDegrees=15, twistLimitDegrees=5}
 {bone1Name=dyn_shoelaceRight_01_r, bone2Name=dyn_shoelaceRight_01_r_MSTR, swing1LimitDegrees=15, swing2LimitDegrees=15, twistLimitDegrees=5}
 {bone1Name=dyn_shoelaceRight_02_r, bone2Name=dyn_shoelaceRight_01_r, swing1LimitDegrees=15, swing2LimitDegrees=15, twistLimitDegrees=5}
 {bone1Name=dyn_boot_01_l, bone2Name=dyn_boot_01_l_MSTR, swing1LimitDegrees=2, swing2LimitDegrees=2, twistLimitDegrees=2}
 {bone1Name=dyn_boot_01_r, bone2Name=dyn_boot_01_r_MSTR, swing1LimitDegrees=2, swing2LimitDegrees=2, twistLimitDegrees=2}
]
